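# Translucent unlit Niagara mesh-particle material for the Arctic Ice Talus backpack FX that renders mesh pieces as soft round sprites.
# World position offset combines the Sprite function with a camera offset, both driven by dynamic parameters.
# Particle color feeds a white tint; particle alpha is multiplied by a UV radial mask and an inverted world-space sphere mask around the offset particle position.
# Both masks use Sphere Mask Radius for hardness too; FlareTexture is unconnected.
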Material /BRCosmetics/Accessories/FORT_Backpacks/Backpack_M_MED_ArcticIceTalus/FX/M_MultiSpriteParticleMesh
graph {
 properties {
  bUsedWithNiagaraMeshParticles=true
  blendMode=BLEND_Translucent
  pixelDepthOffsetMode=PDOM_Legacy
  shadingModel=MSM_Unlit
  usedShadingModels=Unlit
 }
 nodes [
  Constant() {r=0.35}
  n1 = DynamicParameter() {defaultValue=<rgba 10 5 0.5 0.5>, paramNames [SpriteSize, Camera Offset, Sphere Mask Hardness, Sphere Mask Radius]}
  n2 = AppendVector(A=@n1.SpriteSize, B=@n1.SpriteSize)
  n3 = Add(
   A=MaterialFunctionCall!(Center Location=WorldPosition().XYZ, 0-1 UVs, XY scale=Multiply(A=@n2), OverridePivot, "Normals (Optional)") {materialFunction=&Sprite}.World position offset
   B=MaterialFunctionCall(Offset Amount=@n1.Camera Offset, World Position=WorldPosition().XYZ) {materialFunction=/Engine/Functions/Engine_MaterialFunctions02/WorldPositionOffset/CameraOffset}.Clamped Camera Offset
  )
  FeatureLevelSwitch(Default=@n3, ES2=@n3, ES3_1=@n3, SM4_REMOVED=@n3)
  n4 = ParticleColor()
  Multiply(A=Constant3Vector() {constant=<rgba 1 1 1 0>}, B=@n4.RGB)
  n5 = TextureCoordinate()
  n6 = Constant() {r=0.5}
  n7 = DynamicParameter() {defaultValue=<rgba 10 5 0.5 0.5>, paramNames [SpriteSize, Camera Offset, Sphere Mask Hardness, Sphere Mask Radius]}
  Multiply(A=Multiply(A=@n4.A, B=SphereMask(A=@n5, B=@n6, Radius=@n7.Sphere Mask Radius, Hardness=@n7.Sphere Mask Radius)), B=NamedRerouteUsage())
  n8 = WorldPosition()
  n9 = Add(A=ParticlePositionWS(), B=VectorParameter() {parameterName=ParticlePositionOffset, defaultValue=<rgba 0 0 300 0>}.RGB)
  NamedRerouteDeclaration(None=LinearInterpolate(Alpha=SphereMask(A=@n8.XYZ, B=@n9, Radius=@n1.Sphere Mask Radius, Hardness=@n1.Sphere Mask Radius) {attenuationRadius=500}) {constA=1, constB=0}) {name=SphereMask, nodeColor=<rgba 0.6641 1 0 1>}
  TextureSampleParameter2D() {parameterName=FlareTexture, texture=/Game/Global/Textures/Default/Masks_and_Gradients/WM_PinGlow_Sphere}
 ]
}
