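# Idle Niagara effect for the ArcticIceTalus backpack cosmetic, providing a purple glow.
# Only Light001 is enabled: a burst-spawned, offset particle that outlives its lifetime, driving both a light renderer and a mesh renderer with dynamic material parameters.
# Disabled emitters remain: an empty template, an earlier light-only version, and Glow001, which samples an FX static mesh for sprite glows with sine pulsing and time-of-day alpha.
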NiagaraSystem /BRCosmetics/Accessories/FORT_Backpacks/Backpack_M_MED_ArcticIceTalus/FX/NS_Backpack_ArcticIceTalus_Idle
dependencies {
 usedDataInterfaces [&NiagaraDataInterfacePlatformSet, &NiagaraDataInterfaceStaticMesh]
 usedDynamicInputs []
 usedModules [&SystemState, &Timeline]
}
emitters {
 Empty {
  bEnabled=false
  emitterUpdateScript [
   &EmitterState()
   &SpawnRate(SpawnRate=5, Use Spawn Probability=false)
  ]
  particleSpawnScript [&InitializeParticle(
    Color Mode=::NewEnumerator0
    Lifetime=5
    Sprite Size Mode=::NewEnumerator3
    Uniform Sprite Size=100
    UsePositionOffset=false
    Write Material Random=true
   )]
  particleUpdateScript [&ParticleState()]
  rendererClasses [&NiagaraSpriteRendererProperties]
  renderers [{rendererClass=&NiagaraSpriteRendererProperties}]
 }
 Glow001 {
  bEnabled=false
  emitterUpdateScript [
   &EmitterState()
   &SpawnBurst_Instantaneous(Spawn Count=210, Use Spawn Probability=false, UseLoopCountLimit=true)
  ]
  particleSpawnScript [
   &InitializeParticle(
    Color=<rgba 0.4635 0.0473 0.8148 0.1>
    Lifetime=0.25
    Sprite Size=&Multiply_Vector2D_ByFloat(Float=0.2, Vector2D=<35 35>)
    Sprite Size Mode=::NewEnumerator4
    UsePositionOffset=false
    Write Material Random=true
   )
   &CameraOffset(Camera Offset Amount=-1)
   &KillParticles(Kill Particles=&SetBoolByIntComparison(
     A=&ModuloInt(
      A=&ReturnParticleIDAttributes()
      B=&MakeCustomIntFromBool(
       Bool=&IsPlatformSetActive(New Platform Set=DataInterface() {propertyValues {Platforms {qualityLevelMask=-14}}})
       True INT=6
       False INT=&MakeCustomIntFromBool(
        Bool=&IsPlatformSetActive(New Platform Set=DataInterface() {propertyValues {Platforms {qualityLevelMask=-28}}})
        True INT=5
        False INT=4
       )
      )
     )
     B=0
     Comparison Type=::NewEnumerator0
    )) {enabled=false}
  ]
  particleUpdateScript [
   &ParticleState(Kill Particles When Lifetime Has Elapsed=false, Let Infinitely Lived Particles Die When Emitter Deactivates=true)
   /Niagara/Modules/Spawn/Location/StaticMeshLocation(
    FixedBarycentric=true
    MaskSection=false
    MaskSpawnGroup=false
    Offset Position Along Sampled Normal=0
    OffsetAlongNormal=true
    SamplePosOffset=true
    Static Mesh=DataInterface() {propertyValues {
      DefaultMesh=/Game/Accessories/FORT_Backpacks/Backpack_M_MED_ArcticIceTalus/FX/SM_BackpackArcticIceTalus_Fx
      InstanceIndex=-1
      LODIndex=0
      PreviewMesh=&SM_BackpackArcticIceTalus_Fx
      SoftSourceActor=null
      SourceMode=DefaultMeshOnly
      bAllowSamplingFromStreamingLODs=false
      bCaptureTransformsPerFrame=true
      bUsePhysicsBodyVelocity=false
      FilteredSockets []
      LODIndexUserParameter {parameter {name=None, typeDefHandle {registeredTypeIndex=108}, varData []}}
      MeshParameterBinding {parameter {name=None, typeDefHandle {registeredTypeIndex=1}, varData []}}
      SectionFilter {allowedMaterialSlots []}
     }}
    Triangle ID:NiagaraInt32=&ReturnParticleIDAttributes(Particle ID Attribute=::NewEnumerator1)
    Triangle Sampling Mode:ENiagara_StaticTriangleMode=::NewEnumerator2
    Write To Particle Position=false
   )
   /SetVariables(Particles.OffPosition:NiagaraPosition=&AddVectorToPosition(
     Position=$Particles.StaticMeshLocation.SampledPosition
     Vector=&Multiply_VectorByFloat(Vector=$Output.StaticMeshLocation.SampledTangent, Float=0)
    )) {isSetParameters=true}
   /SetVariables(Particles.OffPosition:NiagaraPosition=$Particles.StaticMeshLocation.SampledPosition) {enabled=false, isSetParameters=true}
   &ScaleColor(
    Scale Alpha=&Multiply_Float(
     A=&Lerp_Float(A=0.5, B=1, Alpha=&Sine(Normalized Angle=$Emitter.Age, Period=3, Scale=0.5, Bias=0.5))
     B=&Multiply_Float(A=&FNTimeOfDayScalar(Night Scale=0.75, Morning Scale=1, Evening Scale=1), B=1)
    )
    ScaleA=true
    ScaleRGB=true
   )
   /SetVariables(Particles.Velocity=$Output.StaticMeshLocation.SampledTangent) {enabled=false, isSetParameters=true}
   &Color(Color=&LerpLinearColors(
     StartColor=<rgba 1 1 1 1>
     EndColor=<rgba 0 0 0 1>
     LerpFactor=&MakeFloatFromInt(Int=&MakeCustomIntFromBool(
       Bool=&SetBoolByIntComparison(A=&ReturnParticleIDAttributes(), B=74, Comparison Type=::NewEnumerator4)
       True INT=0
       False INT=1
      ))
    )) {enabled=false}
  ]
  rendererClasses [&NiagaraSpriteRendererProperties]
  renderers [{rendererClass=&NiagaraSpriteRendererProperties}]
 }
 Light {
  bEnabled=false
  emitterUpdateScript [
   &EmitterState(Life Cycle Mode=::NewEnumerator1, Loop Behavior=::NewEnumerator1, Loop Duration Mode=::NewEnumerator1, UseLoopDelay=false)
   &SpawnBurst_Instantaneous(Use Spawn Probability=false, UseLoopCountLimit=false)
  ]
  particleSpawnScript [&InitializeParticle(Color Mode=::NewEnumerator0, Lifetime=5, Position Offset=<0 -5 0>, UsePositionOffset=true, Write Material Random=true)]
  particleUpdateScript [
   &ParticleState(Kill Particles When Lifetime Has Elapsed=false, Let Infinitely Lived Particles Die When Emitter Deactivates=true)
   &FN_Light_Attributes(
    Color=<rgba 0.5063 0 1 1>
    Create Color Attribute=true
    Create Exponent Attribute=true
    Create Radius Attribute=true
    Exponent=8
    Radius=75
    Scale Color=3
   )
  ]
  rendererClasses [&NiagaraLightRendererProperties]
  renderers [{rendererClass=&NiagaraLightRendererProperties, rendererIndex=0}]
 }
 Light001 {
  emitterUpdateScript [
   &EmitterState(Life Cycle Mode=::NewEnumerator1, Loop Behavior=::NewEnumerator1, Loop Duration Mode=::NewEnumerator1, UseLoopDelay=false)
   &SpawnBurst_Instantaneous(Use Spawn Probability=false, UseLoopCountLimit=false)
  ]
  particleSpawnScript [&InitializeParticle(
    Color=<rgba 0.4635 0.0473 0.8148 0.1>
    Lifetime=5
    Mesh Scale=<-1 1 -1>
    Mesh Scale Mode=::NewEnumerator4
    Position Offset=<0 -4.5 0>
    UsePositionOffset=true
    Write Material Random=true
   )]
  particleUpdateScript [
   &ParticleState(Kill Particles When Lifetime Has Elapsed=false, Let Infinitely Lived Particles Die When Emitter Deactivates=true)
   &FN_Light_Attributes(
    Color=<rgba 0.5063 0 1 1>
    Create Color Attribute=true
    Create Exponent Attribute=true
    Create Radius Attribute=true
    Exponent=8
    Radius=75
    Scale Color=3
   )
   &DynamicMaterialParameters(
    Index 0 Param 1=2
    Index 0 Param 2=1
    Index 0 Param 4=0.5
    Module.Param0WriteEnabled=true
    Module.Param1WriteEnabled=true
    Module.Param2WriteEnabled=true
    Module.Param3WriteEnabled=true
   )
  ]
  rendererClasses [&NiagaraLightRendererProperties, &NiagaraMeshRendererProperties]
  renderers [
   {rendererClass=&NiagaraLightRendererProperties, rendererIndex=0}
   {rendererClass=&NiagaraMeshRendererProperties, rendererIndex=1}
  ]
 }
}
summary {
 systemName=NS_Backpack_ArcticIceTalus_Idle
 emitters [
  {bEnabled=false, emitterName=Glow001, rendererClasses [&NiagaraSpriteRendererProperties]}
  {bEnabled=false, emitterName=Empty, rendererClasses [&NiagaraSpriteRendererProperties]}
  {bEnabled=false, emitterName=Light, rendererClasses [&NiagaraLightRendererProperties]}
  {emitterName=Light001, rendererClasses [&NiagaraLightRendererProperties, &NiagaraMeshRendererProperties]}
 ]
 userVariables []
}
