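# Translucent unlit aura material for the Arctic Ice Talus backpack symbol, built on an inverted-hull overlay outline graph.
# Vertices push along normals by a thickness with optional distance, mask and vertex-color controls, while two-sided sign masks out front faces.
# Emissive is a tinted intensity, and opacity uses inverted, powered Fresnel scaled by pixel depth, minus an inverted mask.
# The attributes feed CharacterDeRez_Transition; the graph also carries roughness edge-detection and debug branches.
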Material /BRCosmetics/Accessories/FORT_Backpacks/Backpack_M_MED_ArcticIceTalus/Materials/M_ArcticIceTalus_Symbol_Aura
graph {
 properties {
  bEnableResponsiveAA=true
  bUseMaterialAttributes=true
  bUsedWithMeshDeformer=true
  bUsedWithSkeletalMesh=true
  blendMode=BLEND_Translucent
  pixelDepthOffsetMode=PDOM_Legacy
  shadingModel=MSM_Unlit
  twoSided=true
  usedShadingModels=Unlit
 }
 nodes [
  ComponentMask() {g=true, r=true}
  n1 = ComponentMask(None=MaterialFunctionCall() {materialFunction=/Engine/Functions/Engine_MaterialFunctions02/Utility/FOV}."FOV ") {g=true}
  Divide(A=@n1)
  n2 = VertexColor()
  n3 = ChannelMaskParameter(None=AppendVector(A=@n2, B=@n2)) {group=OverlayOutlines, maskChannel=Green, parameterName=VertexHide Channel, sortPriority=2, defaultValue=<rgba 0 1 0 0>}
  n4 = ChannelMaskParameter(None=@n2) {group=OverlayOutlines, maskChannel=Green, parameterName=VertexHideMaskChannel, sortPriority=2, defaultValue=<rgba 0 1 0 0>}
  n5 = LinearInterpolate(
   A=SphereMask(
    A=@n3.R
    B=ScalarParameter() {defaultValue=0.1, group=OverlayOutlines, parameterName=VertexHide Threshold, sliderMax=1}
    Radius=Constant() {r=0.13}
    Hardness=Constant() {r=1}
   ) {attenuationRadius=0.1}
   Alpha=@n4.R
  )
  n6 = Multiply(A=NamedRerouteUsage()) {constB=2}
  n7 = ScalarParameter() {defaultValue=0.1, group=OverlayOutlines, parameterName=LineThickness, sliderMax=1, sortPriority=3}
  n8 = StaticSwitchParameter(
   True=LinearInterpolate(
    A=@n7
    B=ScalarParameter() {defaultValue=0.9, group=OverlayOutlines, parameterName=LineThicknessFar, sliderMax=4, sortPriority=3}
    Alpha=NamedRerouteUsage()
   )
   False=@n7
  ) {group=OverlayOutlines, parameterName=UseDistanceFading, sortPriority=0}
  n9 = Multiply(
   A=VertexNormalWS()
   B=StaticSwitchParameter(True=Multiply(A=@n6, B=@n8), False=@n8) {group=OverlayOutlines, parameterName=UseSkinFX_Thickness, sortPriority=0}
  )
  If(A=@n5, B=Constant() {r=0.5}, "A > B"=Divide(A=Constant() {r=1}, B=Constant()) {constB=0, desc=NaN clipping}, "A < B"=@n9)
  n10 = Saturate(None=InverseLinearInterpolate(A=Constant() {r=1500}, B=Constant() {r=2500}, Value=NamedRerouteUsage()))
  n11 = OneMinus(None=@n10)
  n12 = TextureSampleParameter2D() {
   group=OverlayOutlines
   parameterName=SkinFX
   samplerType=SAMPLERTYPE_Masks
   sortPriority=1
   texture=/Game/Global/Textures/Default/Blanks/T_MidgreyMask
  }
  n13 = Add(
   A=NamedRerouteDeclaration(None=ChannelMaskParameter(None=@n12.RGBA) {group=OverlayOutlines, parameterName=OutlineThickness_Channel, sortPriority=2}.R) {name=ThicknessMask, nodeColor=<rgba 1 0.1689 0 1>}
   B=Constant() {r=-0.009}
  )
  n14 = OneMinus(None=TwoSidedSign())
  n15 = StaticSwitchParameter(True=Multiply(A=Ceil(None=@n13) {desc=cut outlines mask}, B=@n14), False=@n14) {group=OverlayOutlines, parameterName=UseSkinFX_Thickness, sortPriority=0}
  LinearInterpolate(
   A=@n15
   B=MaterialFunctionCall(Alpha Threshold=Multiply(A=@n15, B=@n11)) {materialFunction=&DitherTemporalAA}.Result
   Alpha=@n10
  )
  n16 = ComponentMask(None=ViewProperty().Property) {r=true}
  n17 = Constant() {r=57.2958}
  n18 = Saturate(None=Subtract(A=Multiply(A=@n16, B=@n17), B=ScalarParameter() {defaultValue=70, group=FrontEnd, parameterName=FOV Threshold, sliderMax=150, sliderMin=35}))
  LinearInterpolate(
   A=@n8
   B=ScalarParameter() {defaultValue=0.1, group=OverlayOutlines_FrontEnd, parameterName=LineThicknessFrontEnd, sliderMax=1, sortPriority=3}
   Alpha=OneMinus(None=Ceil(None=@n18)) {desc="if true, is frontend FOV"}
  )
  n19 = NamedRerouteUsage()
  n20 = MaterialFunctionCall!(Opacity=@n19) {materialFunction=/Game/Packages/Fortress_SharedMaterials/Base_Material_Functions/BaseCharacterShaderFunctions/MF_TechArt_ChromeOutlineOverride}
  n21 = Fresnel(ExponentIn=ScalarParameter() {defaultValue=0.7, parameterName=FresnelExp}) {baseReflectFraction=0.1, exponent=0.5}
  n22 = PixelDepth()
  n23 = ScalarParameter() {defaultValue=100, parameterName=FresnelDepthDivider}
  n24 = Multiply(
   A=Power(Base=OneMinus(None=@n21), Exp=ScalarParameter() {defaultValue=15, parameterName=FresnelPower}) {constExponent=6}
   B=Add(A=ScalarParameter() {defaultValue=2, parameterName=FresnelBaseIntensity}, B=Divide(A=@n22.R, B=@n23))
  )
  n25 = OneMinus(None=@n19)
  n26 = SetMaterialAttributes!(
   MaterialAttributes
   Emissive Color=Multiply(
    A=VectorParameter() {parameterName=AuraColor, defaultValue=<rgba 0.0919 0.003 0.1615 1>}.RGB
    B=ScalarParameter() {defaultValue=6, parameterName=AuraIntensity}
   )
   Opacity Mask=@n20.Result
   World Position Offset=NamedRerouteUsage()
   Opacity=Subtract(A=Saturate(None=@n24), B=Saturate(None=@n25))
   Normal
  ) {attributeSetTypes [
    #1
    #2
    #3
    #4
    #5
   ]}
  MaterialFunctionCall!(Fresnel, Material=@n26) {materialFunction=/Game/Packages/Fortress_SharedMaterials/Base_Material_Functions/CharacterDeRez_Transition}
  n27 = TextureSampleParameter2D() {
   desc=Texture for Outline Diffuse Color
   group=OverlayOutlines
   parameterName=Diffuse
   sortPriority=1
   texture=&T_BlacksRGB
  }
  n28 = ScalarParameter() {
   defaultValue=1
   desc=Multiplier for the Outline Diffuse Texture
   group=OverlayOutlines
   parameterName=DiffuseMult
   sliderMax=1
   sortPriority=3
  }
  n29 = StaticSwitchParameter(True=Multiply(A=@n27.RGB, B=@n28), False=VectorParameter() {parameterName=Color, defaultValue=<rgba 0 0 0 1>}.RGB) {desc=Enable to use a texture to color the edge outlines of the mesh., group=OverlayOutlines, parameterName=UseDiffuseMap, sortPriority=0}
  MaterialFunctionCall!(In=@n29, LUT, LUT Multiplier) {materialFunction=/Game/Packages/Fortress_SharedMaterials/Anime_Master_Material/MF_ApplyLUT}
  n30 = ScreenPosition()
  MaterialFunctionCall(
   Number=MaterialFunctionCall() {materialFunction=&ScreenResolution}.Visible Resolution
   UVs=@n30.ViewportUV
  ) {materialFunction=&DebugScalarValues}
  n31 = Distance(A=CameraPositionWS(), B=ObjectPositionWS())
  n32 = Divide(A=Multiply(A=@n31, B=@n1), B=ComponentMask(None=ViewSize() {desc=PixelSize}) {g=true})
  MaterialFunctionCall!(Vector2=@n32, MaximumNumberOfDigits, UVs=@n30.ViewportUV, DebugTextLocation RG_UpperRight BA_LowerLeft, Component Spacing) {materialFunction=/Engine/Functions/Engine_MaterialFunctions02/Utility/DebugFloat2Values}
  Multiply(A=Constant() {r=1}, B=@n32)
  Multiply(A=Multiply(A=Sine(), B=Constant() {r=2}), B=@n31)
  n33 = Subtract(A=ObjectPositionWS(), B=WorldPosition().XYZ)
  Multiply(A=@n33, B=@n5)
  n34 = CameraPositionWS()
  n35 = ObjectPositionWS()
  n36 = NamedRerouteDeclaration(None=Subtract(
    A=Distance(A=@n34, B=@n35)
    B=ScalarParameter() {defaultValue=50, group=OverlayOutlines, parameterName=FadeOffset, sliderMax=400, sortPriority=5}
   ) {constB=200}) {name=DistanceAfterOffset, nodeColor=<rgba 0.9631 1 0 1>}
  n37 = ScalarParameter() {defaultValue=500, group=OverlayOutlines, parameterName=FadeDistance, sliderMax=2000, sortPriority=5}
  NamedRerouteDeclaration(None=Saturate(None=Divide(A=@n36, B=@n37) {constB=1000})) {name=DistanceFading, nodeColor=<rgba 1 0.3047 0 1>}
  n38 = Saturate(None=@n5)
  n39 = StaticSwitchParameter(True=LinearInterpolate(A=@n33, B=@n9, Alpha=Round(None=@n38)), False=@n9) {group=OverlayOutlines, parameterName=UseVertexOutlineHiding, sortPriority=0}
  NamedRerouteDeclaration(None=StaticSwitchParameter(
    True=MaterialFunctionCall!(WPO=@n39) {materialFunction=/Game/Packages/Fortress_SharedMaterials/Base_Material_Functions/Fortnite_HideElementsWithVertexColor}.Result
    False=@n39
   ) {group=Opacity Mask, parameterName=Use Vertex Colors for Mask}) {name=WPO, nodeColor=<rgba 1 0.3281 0 1>}
  n40 = TextureSampleParameter2D() {
   group=OverlayOutlines
   parameterName=OutlineMaskTexture
   samplerType=SAMPLERTYPE_Masks
   texture=&T_BlackMask
  }
  n41 = ChannelMaskParameter(None=@n40.RGB) {group=OverlayOutlines, parameterName=OutlineMaskChannelSelector}
  n42 = OneMinus(None=@n41.R)
  n43 = StaticSwitchParameter(True=Multiply(A=StaticSwitchParameter(True=@n41.R, False=@n42) {group=OverlayOutlines, parameterName=InvertOutlineMask}, B=@n15), False=@n15) {
   desc=Exclude a section of the mesh from having Distance Outlines based on a Texture
   group=OverlayOutlines
   parameterName=MaskOutlineByTexture
   sortPriority=1
  }
  NamedRerouteDeclaration(None=@n43) {name=OpacityMask, nodeColor=<rgba 0 0.8125 1 1>}
  n44 = SceneTexture() {sceneTextureId=PPI_Roughness}
  n45 = ComponentMask(None=ScreenPosition().ViewportUV) {g=true, r=true}
  n46 = Multiply(
   A=SceneTexelSize()
   B=ScalarParameter() {defaultValue=1, group=OverlayOutlines Roughness, parameterName=RoughnessOutline_Thickness, sliderMax=8}
  )
  n47 = Constant()
  n48 = AppendVector(A=ComponentMask(None=@n46) {desc=x, r=true}, B=@n47)
  n49 = SceneTexture(UVs=Add(A=@n45, B=@n48) {desc=x+}) {sceneTextureId=PPI_Roughness}
  n50 = AppendVector(A=@n47, B=ComponentMask(None=@n46) {desc=y, g=true}) {desc=y+}
  n51 = SceneTexture(UVs=Add(A=@n45, B=@n50) {desc=y+}) {sceneTextureId=PPI_Roughness}
  n52 = Constant() {r=-1}
  n53 = SceneTexture(UVs=Add(A=@n45, B=Multiply(A=@n48, B=@n52)) {desc=x-}) {sceneTextureId=PPI_Roughness}
  n54 = SceneTexture(UVs=Add(A=@n45, B=Multiply(A=@n52, B=@n50)) {desc=y-}) {sceneTextureId=PPI_Roughness}
  n55 = Add(
   A=Add(A=Subtract(A=@n44.Color, B=@n49.Color), B=Subtract(A=@n44.Color, B=@n51.Color))
   B=Add(A=Subtract(A=@n44.Color, B=@n53.Color), B=Subtract(A=@n44.Color, B=@n54.Color))
  )
  n56 = Add(A=@n45, B=Multiply(A=@n46, B=Constant2Vector() {g=1, r=1})) {desc=x+}
  n57 = Add(A=@n45, B=Multiply(A=@n46, B=Constant2Vector() {g=1, r=-1})) {desc=y+}
  n58 = Add(
   A=Subtract(A=@n44.Color, B=SceneTexture(UVs=@n56) {sceneTextureId=PPI_Roughness}.Color)
   B=Subtract(A=@n44.Color, B=SceneTexture(UVs=@n57) {sceneTextureId=PPI_Roughness}.Color)
  )
  n59 = Add(A=@n45, B=Multiply(A=@n46, B=Constant2Vector() {g=-1, r=1})) {desc=x-}
  n60 = Add(A=@n45, B=Multiply(A=@n46, B=Constant2Vector() {g=-1, r=-1})) {desc=y-}
  n61 = Add(
   A=Subtract(A=@n44.Color, B=SceneTexture(UVs=@n59) {sceneTextureId=PPI_Roughness}.Color)
   B=Subtract(A=@n44.Color, B=SceneTexture(UVs=@n60) {sceneTextureId=PPI_Roughness}.Color)
  )
  n62 = StaticSwitchParameter(True=Add(A=@n55, B=Add(A=@n58, B=@n61)), False=@n55) {group=OverlayOutlines Roughness, parameterName=UseRoughnessSamples, sortPriority=0}
  n63 = ScalarParameter() {defaultValue=4, group=OverlayOutlines Roughness, parameterName=Divisor, sliderMax=16}
  n64 = Multiply(
   A=Saturate(None=Divide(A=@n62, B=@n63))
   B=ScalarParameter() {defaultValue=1, group=OverlayOutlines Roughness, parameterName=Mult, sliderMax=16}
  )
  NamedRerouteDeclaration(None=Round(None=Saturate(None=@n64))) {name=RoughnessOutlines, nodeColor=<rgba 0 0.5078 1 1>}
  Sine(None=ViewProperty().Property)
  n65 = Multiply(A=ScreenPosition().ViewportUV) {constB=8}
  n66 = NamedRerouteUsage()
  n67 = Constant3Vector() {constant=<rgba 0 0.2 1 0>}
  n68 = Constant()
  n69 = LinearInterpolate(A=Constant3Vector() {constant=<rgba 0.2 0 0 0>}, B=Constant3Vector() {constant=<rgba 0 0.2 0 0>}, Alpha=@n66)
  n70 = Constant3Vector() {constant=<rgba 0.2 0 1 0>}
  n71 = Add(
   A=MaterialFunctionCall(Number=NamedRerouteUsage(), MaximumNumberOfDigits=Constant() {r=6}, UVs=Frac(None=@n65)) {materialFunction=&DebugScalarValues}.Result
   B=If(A=@n66, B=Constant() {r=1}, "A > B"=@n67, "A == B"=@n67, "A < B"=If(A=@n66, B=@n68, "A > B"=@n69, "A == B"=@n70, "A < B"=@n70))
  )
  StaticSwitchParameter(True=SetMaterialAttributes!(MaterialAttributes, Emissive Color=Saturate(None=@n71), World Position Offset=NamedRerouteUsage()) {attributeSetTypes [#1, #3]}) {group=OverlayOutlines Debug, parameterName=DebugFadeDistance, sortPriority=64}
  StaticSwitchParameter(True=Add(A=NamedRerouteUsage())) {group=OverlayOutlines Roughness, parameterName=UseRoughnessOutlines, sortPriority=0}
  n72 = TextureSampleParameter2D() {
   group=OverlayOutlines
   parameterName=Normal
   samplerType=SAMPLERTYPE_Normal
   sortPriority=1
   texture=&FlatNormalBlue
  }
  StaticSwitchParameter(
   True=LinearInterpolate(
    A=VertexNormalWS()
    B=Transform(None=@n72.RGB)
    Alpha=ScalarParameter() {group=OverlayOutlines, parameterName=BlendVertexPixelNormal, sliderMax=1, sortPriority=5}
   )
   False=VertexNormalWS()
  ) {group=OverlayOutlines, parameterName=UseNormalMap, sortPriority=0}
  StaticSwitchParameter(True=SetMaterialAttributes!(MaterialAttributes, Emissive Color=NamedRerouteUsage(), Opacity=Constant() {r=1}) {attributeSetTypes [#1, #4]}) {group=OverlayOutlines Debug, parameterName=DebugRoughnessOutlines, sortPriority=64}
  Tangent!(None)
  ViewProperty() {property=MEVP_ViewportOffset}
  ViewProperty() {property=MEVP_ViewSize}
  ViewProperty() {property=MEVP_BufferSize}
  ViewSize()
 ]
}
