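# Animation Blueprint for the ArcticIceTalus cosmetic backpack that makes its pack bone bob and react to player movement.
# Spring-based AnimDynamics feeds chained additive bone offsets for crouched movement, idle floating, tactical sprinting and skydiving.
# The event graph builds the float from layered sine waves, scales crouch correction by speed, and eases physics weight lower while airborne.
# Separate jump and fall offset nodes are not wired to the output pose.
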AnimBlueprint /BRCosmetics/Accessories/FORT_Backpacks/Backpack_M_MED_ArcticIceTalus/Meshes/M_MED_ArcticIceTalus_Pack_AnimBP
target_skeleton=&/M_MED_ArcticIceTalus_Pack_Skeleton
graphs [
 AnimGraph {nodes [
   n1 = "Transform (Modify) Bone - Bone: pack"!(ComponentPose, bAlphaBoolEnabled:bool=Get bIsJumping(), Translation="0,0,-0.3", Rotation="0,0,0", Scale) -> (Pose) {
    alphaInputType=Bool
    translationMode=BMM_Additive
    translationSpace=BCS_BoneSpace
    alphaBoolBlend {blendInTime=0.3, blendOption=ExpOut, blendOutTime=0.1}
    boneToModify {boneName=pack}
    translation {z=-0.3}
   }
   "Transform (Modify) Bone - Bone: pack"(ComponentPose=@n1, bAlphaBoolEnabled=Get bIsFalling(), Translation="0,0,0.2", Rotation="0,0,0") {
    alphaInputType=Bool
    translationMode=BMM_Additive
    translationSpace=BCS_BoneSpace
    alphaBoolBlend {blendInTime=0.1, blendOption=ExpInOut, blendOutTime=0.3}
    boneToModify {boneName=pack}
    translation {z=0.2}
   }
   n2 = Anim Dynamics - Bone: pack!(
    ComponentPose=Local To Component(LocalPose=Input Pose())
    Alpha=Get AlphaJump() -> (AlphaJump:double)
    GravityOverride:Vector="0,0,50"
    ComponentLinearAccScale:Vector="0.1,0.1,0.1"
    ComponentLinearVelScale:Vector="0.12,0.12,0.12"
   ) -> (Pose) {
    alpha=0
    angularBiasOverride=1
    angularDampingOverride=0.9
    angularSpringConstant=20
    bLinearSpring=true
    bOverrideLinearDamping=true
    bUseGravityOverride=true
    bUsePlanarLimit=false
    bUseSphericalLimits=true
    linearDampingOverride=0.81
    linearSpringConstant=25
    simulationSpace=BoneRelative
    alphaBoolBlend {blendOption=Linear}
    boundBone {boneName=pack}
    componentAppliedLinearAccClamp=<800 800 800>
    physicsBodyDefinitions [{
      sphereCollisionRadius=0
      boundBone {boneName=pack}
      boxExtents=<50 20 50>
      constraintSetup {
       linearZLimitType=Free
       angularLimitsMax=<0 0 0>
       angularLimitsMin=<0 0 0>
       angularTarget=<1 0 0>
       linearAxesMax=<19 0 10>
       linearAxesMin=<-10 0 -5>
      }
      localJointOffset=<1 0 0>
     }]
    relativeSpaceBone {boneName=root}
    sphericalLimits [{limitRadius=7, limitType=Inner, drivingBone {boneName=pack}, sphereLocalOffset=<0 0 4.4>}]
   }
   n3 = "Transform (Modify) Bone - Bone: pack"!(
    ComponentPose=@n2
    Alpha:float=Get CrouchMovementCorrection() -> (CrouchMovementCorrection:double)
    Translation="0,0,-4"
    Rotation="0,0,0"
    Scale
   ) -> (Pose) {
    translationMode=BMM_Additive
    translationSpace=BCS_BoneSpace
    alphaBoolBlend {blendInTime=0.3, blendOption=QuarticInOut, blendOutTime=0.2}
    alphaScaleBiasClamp {bInterpResult=true, interpSpeedDecreasing=5, interpSpeedIncreasing=5}
    boneToModify {boneName=pack}
    translation {z=-4}
   }
   n4 = "Transform (Modify) Bone - Bone: pack"!(ComponentPose=@n3, bAlphaBoolEnabled:bool, Translation=Get Backpack_floating() -> (Backpack_floating:Vector), Rotation="0,0,0", Scale) -> (Pose) {
    alphaInputType=Bool
    rotationMode=BMM_Additive
    rotationSpace=BCS_BoneSpace
    translationMode=BMM_Additive
    translationSpace=BCS_ParentBoneSpace
    alphaBoolBlend {blendInTime=0.8, blendOption=Linear, blendOutTime=0.8}
    boneToModify {boneName=pack}
   }
   n5 = "Transform (Modify) Bone - Bone: pack"(ComponentPose=@n4, bAlphaBoolEnabled=Get bIsTacticalSprinting(), Translation="0,-10,0", Rotation="0,0,0") {
    alphaInputType=Bool
    rotationMode=BMM_Additive
    rotationSpace=BCS_BoneSpace
    translationMode=BMM_Additive
    translationSpace=BCS_BoneSpace
    alphaBoolBlend {blendInTime=0.2, blendOption=Cubic, blendOutTime=0.8}
    boneToModify {boneName=pack}
    translation {y=-10}
   }
   n6 = "Transform (Modify) Bone - Bone: pack"(ComponentPose=@n5, bAlphaBoolEnabled=Get bIsSkyDiving(), Translation="0,-5,5", Rotation="0,0,5") {
    alphaInputType=Bool
    rotationMode=BMM_Replace
    rotationSpace=BCS_BoneSpace
    translationMode=BMM_Additive
    translationSpace=BCS_BoneSpace
    alphaBoolBlend {blendInTime=0.2, blendOption=Linear, blendOutTime=0.2}
    boneToModify {boneName=pack}
    rotation {roll=5}
    translation {y=-5, z=5}
   }
   n7 = "Transform (Modify) Bone - Bone: pack"(ComponentPose=@n6, Translation="0,-3,0", Rotation="0,0,0") {
    alphaInputType=Bool
    rotationMode=BMM_Replace
    rotationSpace=BCS_BoneSpace
    translationMode=BMM_Additive
    translationSpace=BCS_BoneSpace
    alphaBoolBlend {blendInTime=0.2, blendOption=Linear, blendOutTime=0.2}
    boneToModify {boneName=pack}
    translation {y=-3}
   }
   Output Pose(Result=Component To Local(ComponentPose=@n7)) {layerGroup=None}
  ]}
 EventGraph {nodes [
   n1 = Event BlueprintUpdateAnimation()
   > Set Delta Time X(Delta Time X=@n1.DeltaTimeX)
   n2 = float * float(A=GetGameTimeInSeconds(), B=1.2) {is=title K2Node_PromotableOperator}
   n3 = "Sin (Radians)"(A=float * float(A=1.21, B=@n2) {is=title K2Node_CommutativeAssociativeBinaryOperator})
   n4 = "Sin (Radians)"(A=float * float(A=0.56, B=@n2))
   n5 = "Sin (Radians)"(A=float * float(A=0.8, B=@n2))
   > Set Backpack_floating(Backpack_floating:Vector=MakeVector(X=float * float(A=@n4, B=0.5), Y=float * float(A=@n3, B=0.15), Z=float * float(A=@n5, B=1))) -> (Output_Get:Vector)
   n6 = "To Float (Boolean)"(InBool=Get bIsCrouching())
   n7 = Vector Length(A=Get PawnVelocity())
   > Set CrouchMovementCorrection(CrouchMovementCorrection:double=MapRangeClamped(Value=float * float(A=@n7, B=@n6) {is=title K2Node_PromotableOperator}, InRangeA=0, InRangeB=420)) -> (Output_Get:double)
   > n8 = Branch(Condition=OR Boolean(A=Get bIsJumping(), B=Get bIsFalling()))
   > Set AlphaJump(AlphaJump:double=FInterpTo(Current=Get AlphaJump(), Target=0.3, DeltaTime=Get Delta Time X(), InterpSpeed=4)) -> (Output_Get:double)
   Set AlphaJump(execute=@n8.else, AlphaJump=FInterpTo(Current=Get AlphaJump(), Target=0.7, DeltaTime=Get Delta Time X(), InterpSpeed=5))
  ]}
]
variables [
 AlphaJump {type=number}
 Backpack_floating {type=Vector}
 CrouchMovementCorrection {type=number}
 Delta Time X {type=number}
]
